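# Translucent unlit fake fog material for sprite and mesh particles, tinted and faded by particle color.
# A world-aligned noise pack, partly following object position and distorted by panning noise, supplies a pseudo-normal for view-based emissive shading and a density mask.
# Density combines squared facing falloff, depth fade and camera-proximity fade, then is squared, ramp-masked and slightly eroded by scrolling noise.
# ES2/ES3_1 feature levels skip the depth and camera fades.
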Material /Game/Effects/Fort_Effects/Effects/FakeLights/M_FakeFog
graph {
 properties {
  bUsedWithMeshParticles=true
  bUsedWithNiagaraMeshParticles=true
  bUsedWithParticleSprites=true
  blendMode=BLEND_Translucent
  pixelDepthOffsetMode=PDOM_Legacy
  shadingModel=MSM_Unlit
  translucencyLightingMode=TLM_VolumetricDirectional
  usedShadingModels=Unlit
 }
 nodes [
  n1 = Constant3Vector() {constant=<rgba 0 0 1 0>}
  n2 = WorldPosition()
  n3 = ScalarParameter() {defaultValue=0.8, parameterName=WorldMotionSize}
  n4 = Multiply(A=ObjectPositionWS(), B=@n3) {constB=0.8}
  n5 = TextureCoordinate() {uTiling=2}
  n6 = Subtract(A=TextureSample(UVs=Panner(Coordinate=@n5) {speedX=0.1, speedY=0.1}) {samplerType=SAMPLERTYPE_LinearColor, texture=&HighContrastNoise}.RGB) {constB=0.5}
  n7 = ScalarParameter() {defaultValue=8, parameterName=Twest}
  n8 = MaterialFunctionCall(
   TextureObject=TextureObject() {samplerType=SAMPLERTYPE_LinearColor, texture=&T_NoisePack_LinearColor}
   TextureSize=VectorParameter() {parameterName=Size, defaultValue=<rgba 100 100 50 1>}.RGB
   WorldPosition=Add(A=Subtract(A=@n2.XYZ, B=@n4), B=Multiply(A=@n6, B=@n7) {constB=64})
   ProjectionTransitionContrast=Constant() {r=-0.5}
  ) {materialFunction=&WorldAlignedTexture}
  n9 = Normalize(VectorInput=AppendVector(A=ComponentMask(None=@n8.XYZ Texture) {g=true, r=true}, B=Constant() {r=1}))
  n10 = OneMinus(None=DotProduct(A=Transform(None=@n1) {transformSourceType=TRANSFORMSOURCE_View}, B=Transform(None=@n9)))
  n11 = ParticleColor()
  MaterialFunctionCall(Emissive=Multiply(A=Power(Base=@n10) {constExponent=0.5}, B=@n11.RGB)) {materialFunction=&MF_FX_EmissiveAdjustments}
  n12 = DotProduct(A=VertexNormalWS(), B=CameraVectorWS())
  n13 = WorldPosition()
  n14 = CameraPositionWS()
  n15 = ScalarParameter() {defaultValue=75, parameterName=CameraFade}
  n16 = Constant() {r=0.25}
  n17 = Multiply(
   A=DepthFade(FadeDistance=ScalarParameter() {defaultValue=100, parameterName=DepehFade})
   B=OneMinus(None=SphereMask(A=@n13.XYZ, B=@n14, Radius=@n15, Hardness=@n16))
  )
  n18 = Constant() {r=1}
  n19 = Multiply(
   A=Multiply(A=ComponentMask(None=@n8.XYZ Texture) {b=true}, B=Multiply(A=@n12, B=@n12))
   B=FeatureLevelSwitch(Default=@n17, ES2=@n18, ES3_1=@n18, SM4_REMOVED=@n17)
  )
  n20 = Multiply(A=@n19, B=@n19)
  n21 = TextureSample() {texture=/Game/Effects/Fort_Effects/Textures/Ramps/T_Ramp_Smooth}
  n22 = TextureSample(UVs=Panner(Coordinate=TextureCoordinate() {uTiling=4, vTiling=2}) {speedX=-0.1, speedY=1}) {samplerType=SAMPLERTYPE_Grayscale, texture=&LowResBlurredNoise}
  Multiply(A=Subtract(A=Multiply(A=@n20, B=@n21.R), B=LinearInterpolate(Alpha=@n22.R) {constA=0.0039, constB=0.0117}), B=@n11.A)
  OneMinus(None=@n3)
 ]
}
