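# M_FakeFog_Cheap is an additive, unlit particle material for sprites and meshes that fakes fog.
# A world-aligned smoke texture, offset by object position and a panned normal-map distortion, supplies color from red-green and a mask from blue.
# View-facing falloff, sphere-mask camera fade (skipped on mobile), particle alpha and depth fade shape the mask; particle color tints emissive.
# Dynamic parameters drive tiling, speed, distortion and fade distances.
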Material /Game/Effects/Fort_Effects/Effects/FakeLights/M_FakeFog_Cheap
graph {
 properties {
  bUsedWithMeshParticles=true
  bUsedWithNiagaraMeshParticles=true
  bUsedWithParticleSprites=true
  blendMode=BLEND_Additive
  pixelDepthOffsetMode=PDOM_Legacy
  shadingModel=MSM_Unlit
  translucencyLightingMode=TLM_VolumetricDirectional
  usedShadingModels=Unlit
 }
 nodes [
  n1 = DynamicParameter() {
   parameterIndex=2
   defaultValue=<rgba 100 1 500 25>
   paramNames [WorldTextureSizeMult, DistortAmount, CameraFadeDistance, DistortedTextureMult]
  }
  n2 = WorldPosition()
  n3 = DynamicParameter() {parameterIndex=1, defaultValue=<rgba 0.1 0.1 1 0.8>, paramNames [XSpeed, YSpeed, TileMult, DistortTextureTiling]}
  n4 = Multiply(A=ObjectPositionWS(), B=@n3.DistortTextureTiling) {constB=0.8}
  n5 = DynamicParameter() {parameterIndex=1, defaultValue=<rgba 0.1 0.1 1 0.8>, paramNames [XSpeed, YSpeed, TileMult, DistortTextureTiling]}
  n6 = Multiply(A=@n5.TileMult, B=TextureCoordinate() {uTiling=2})
  n7 = DynamicParameter() {
   parameterIndex=2
   defaultValue=<rgba 100 1 500 25>
   paramNames [WorldTextureSizeMult, DistortAmount, CameraFadeDistance, DistortedTextureMult]
  }
  n8 = DynamicParameter() {parameterIndex=1, defaultValue=<rgba 0.1 0.1 1 0.8>, paramNames [XSpeed, YSpeed, TileMult, DistortTextureTiling]}
  n9 = TextureSampleParameter2D(UVs=Panner(Coordinate=Multiply(A=@n6, B=@n7.DistortAmount), Speed=AppendVector(A=@n8.XSpeed, B=@n8.XSpeed)) {speedX=0.1, speedY=0.1}) {
   group=Textures
   parameterName=TilingFog_World_Distort
   samplerType=SAMPLERTYPE_Normal
   texture=/Game/Global/Textures/Default/Tileable_Patterns/2ndInfectedMapNormal
  }
  n10 = Subtract(A=@n9.RGB) {constB=0.5}
  n11 = DynamicParameter() {
   parameterIndex=2
   defaultValue=<rgba 100 1 500 25>
   paramNames [WorldTextureSizeMult, DistortAmount, CameraFadeDistance, DistortedTextureMult]
  }
  n12 = MaterialFunctionCall(
   TextureObject=TextureObjectParameter() {group=Textures, parameterName=TilingFog_World, texture=/Game/Global/Textures/Default/Tileable_Patterns/T_Smoke_Puff_Tile_Up}
   TextureSize=@n1.WorldTextureSizeMult
   WorldPosition=Add(A=Subtract(A=@n2.XYZ, B=@n4), B=Multiply(A=@n10, B=@n11.DistortedTextureMult) {constB=64})
   ProjectionTransitionContrast=Constant() {r=-0.5}
  ) {materialFunction=&WorldAlignedTexture}
  n13 = DotProduct(A=VertexNormalWS(), B=CameraVectorWS())
  n14 = Multiply(A=@n13, B=@n13)
  n15 = WorldPosition()
  n16 = CameraPositionWS()
  n17 = DynamicParameter() {
   parameterIndex=2
   defaultValue=<rgba 100 1 500 25>
   paramNames [WorldTextureSizeMult, DistortAmount, CameraFadeDistance, DistortedTextureMult]
  }
  n18 = Constant() {r=0.25}
  n19 = Multiply(A=Constant() {r=1}, B=OneMinus(None=SphereMask(A=@n15.XYZ, B=@n16, Radius=@n17.CameraFadeDistance, Hardness=@n18)))
  n20 = Constant() {r=1}
  n21 = Multiply(
   A=Multiply(A=ComponentMask(None=@n12.XYZ Texture) {b=true}, B=Power(Base=@n14))
   B=FeatureLevelSwitch(Default=@n19, ES2=@n20, ES3_1=@n20, SM4_REMOVED=@n19)
  )
  n22 = Multiply(A=Multiply(A=@n21, B=@n21), B=ParticleColor().A)
  DepthFade(Opacity=@n22, FadeDistance=DynamicParameter() {defaultValue=<rgba 0.1 0.1 1 0.8>, paramNames [XSpeed,YSpeed,TileMult,DepthFade]}.DepthFade)
  n23 = ComponentMask(None=@n12.XYZ Texture) {g=true, r=true}
  n24 = Constant() {r=1}
  n25 = MaterialFunctionCall(In=Normalize(VectorInput=AppendVector(A=@n23, B=@n24)), Contrast=ScalarParameter() {defaultValue=1}) {materialFunction=&CheapContrast}
  MaterialFunctionCall(Emissive=Multiply(A=Power(Base=@n25.Result) {constExponent=0.5}, B=ParticleColor().RGB)) {materialFunction=&MF_FX_EmissiveAdjustments}
  MaterialFunctionCall(In=@n22, Contrast=ScalarParameter() {parameterName="Contrast (S)"}) {materialFunction=&CheapContrast}
 ]
}
