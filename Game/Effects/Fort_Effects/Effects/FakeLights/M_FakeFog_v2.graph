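# Translucent unlit fake fog material for static meshes or Niagara mesh particles, with static switches choosing scalar or dynamic parameters.
# Panning noise distorts a world-aligned noise, partly offset by object position.
# Its RG becomes a pseudo-normal shading the emissive color.
# Its blue channel drives opacity, shaped by facing falloff, depth and camera fades (skipped on mobile), a ramp, subtracted local noise and time-of-day values.
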Material /Game/Effects/Fort_Effects/Effects/FakeLights/M_FakeFog_v2
graph {
 properties {
  bUsedWithNiagaraMeshParticles=true
  blendMode=BLEND_Translucent
  pixelDepthOffsetMode=PDOM_Legacy
  shadingModel=MSM_Unlit
  translucencyLightingMode=TLM_VolumetricDirectional
  usedShadingModels=Unlit
 }
 nodes [
  DynamicParameter() {defaultValue=<rgba 0.1 0.1 1 0.8>, paramNames [XSpeed,YSpeed,TileMult,DepthFade]}
  n1 = VectorParameter() {parameterName=Color, defaultValue=<rgba 1 1 1 1>}
  n2 = ParticleColor()
  n3 = Constant3Vector() {constant=<rgba 0 0 1 0>}
  n4 = TextureObjectParameter() {
   group=Textures
   parameterName=TilingFog_World
   samplerType=SAMPLERTYPE_LinearColor
   texture=&HighContrastNoise
  }
  n5 = DynamicParameter() {parameterIndex=2, defaultValue=<rgba 1 1 75 100>, paramNames [WorldTextureSizeMult, DistortAmount, CameraFadeDistance, DistortedTextureMult]}
  n6 = WorldPosition()
  n7 = DynamicParameter() {parameterIndex=1, defaultValue=<rgba 0.1 0.1 1 0.8>, paramNames [XSpeed, YSpeed, TileMult, DistortTextureTiling]}
  n8 = Multiply(
   A=ObjectPositionWS()
   B=StaticSwitchParameter(True=@n7.DistortTextureTiling, False=ScalarParameter() {defaultValue=0.8, parameterName=WorldMotionSize}) {parameterName=UseNiagaraDynamicParameter?}
  ) {constB=0.8}
  n9 = DynamicParameter() {parameterIndex=1, defaultValue=<rgba 0.1 0.1 1 0.8>, paramNames [XSpeed, YSpeed, TileMult, DistortTextureTiling]}
  n10 = Multiply(
   A=StaticSwitchParameter(True=@n9.TileMult, False=ScalarParameter() {defaultValue=1, parameterName=TileMult}) {parameterName=UseNiagaraDynamicParameter?}
   B=TextureCoordinate() {uTiling=2}
  )
  n11 = DynamicParameter() {parameterIndex=2, defaultValue=<rgba 1 1 75 100>, paramNames [WorldTextureSizeMult, DistortAmount, CameraFadeDistance, DistortedTextureMult]}
  n12 = StaticSwitchParameter(True=@n11.DistortAmount, False=Constant() {r=1}) {parameterName=UseNiagaraDynamicParameter?}
  n13 = DynamicParameter() {parameterIndex=1, defaultValue=<rgba 0.1 0.1 1 0.8>, paramNames [XSpeed, YSpeed, TileMult, DistortTextureTiling]}
  n14 = AppendVector(
   A=StaticSwitchParameter(True=@n13.XSpeed, False=ScalarParameter() {defaultValue=0.1, parameterName=XSpeed}) {parameterName=UseNiagaraDynamicParameter?}
   B=StaticSwitchParameter(True=@n13.YSpeed, False=ScalarParameter() {defaultValue=0.1, parameterName=YSpeed}) {parameterName=UseNiagaraDynamicParameter?}
  )
  n15 = StaticSwitchParameter(
   True=ScalarParameter() {defaultValue=1, parameterName=SpeedMult}
   False=DynamicParameter() {defaultValue=<rgba 0.1 0.1 1 0.8>, paramNames [XSpeed,YSpeed,TileMult,DepthFade]}.XSpeed
  ) {defaultValue=true, parameterName=UseWithStaticMeshes?}
  n16 = TextureSampleParameter2D(UVs=Panner(Coordinate=Multiply(A=@n10, B=@n12), Speed=Multiply(A=@n14, B=@n15)) {speedX=0.1, speedY=0.1}) {
   group=Textures
   parameterName=TilingFog_World_Distort
   samplerType=SAMPLERTYPE_LinearColor
   texture=&HighContrastNoise
  }
  n17 = Subtract(A=@n16.RGB) {constB=0.5}
  n18 = DynamicParameter() {parameterIndex=2, defaultValue=<rgba 1 1 75 100>, paramNames [WorldTextureSizeMult, DistortAmount, CameraFadeDistance, DistortedTextureMult]}
  n19 = StaticSwitchParameter(True=@n18.DistortedTextureMult, False=ScalarParameter() {defaultValue=8, parameterName=DistortAmountTextureMult}) {parameterName=UseNiagaraDynamicParameter?}
  n20 = MaterialFunctionCall(
   TextureObject=@n4
   TextureSize=Multiply(A=VectorParameter() {parameterName=Size, defaultValue=<rgba 100 100 50 1>}.RGB, B=@n5.WorldTextureSizeMult)
   WorldPosition=Add(A=Subtract(A=@n6.XYZ, B=@n8), B=Multiply(A=@n17, B=@n19) {constB=64})
   ProjectionTransitionContrast=Constant() {r=-0.5}
  ) {materialFunction=&WorldAlignedTexture}
  n21 = Normalize(VectorInput=AppendVector(A=ComponentMask(None=@n20.XYZ Texture) {g=true, r=true}, B=Constant() {r=1}))
  n22 = OneMinus(None=DotProduct(A=Transform(None=@n3) {transformSourceType=TRANSFORMSOURCE_View}, B=Transform(None=@n21)))
  MaterialFunctionCall(Emissive=Multiply(
    A=Power(Base=@n22) {constExponent=0.5}
    B=StaticSwitchParameter(True=@n1.RGB, False=@n2.RGB) {defaultValue=true, parameterName=UseWithStaticMeshes?}
   )) {materialFunction=&MF_FX_EmissiveAdjustments}
  n23 = ScalarParameter() {defaultValue=1, parameterName=OpactiyMult}
  n24 = ParticleColor()
  n25 = DotProduct(A=VertexNormalWS(), B=CameraVectorWS())
  n26 = DepthFade(FadeDistance=StaticSwitchParameter(
    True=DynamicParameter() {defaultValue=<rgba 0.1 0.1 1 0.8>, paramNames [XSpeed,YSpeed,TileMult,DepthFade]}.DepthFade
    False=ScalarParameter() {defaultValue=100, parameterName=DepthFade}
   ) {parameterName=UseNiagaraDynamicParameter?})
  n27 = DynamicParameter() {parameterIndex=2, defaultValue=<rgba 1 1 75 100>, paramNames [WorldTextureSizeMult, DistortAmount, CameraFadeDistance, DistortedTextureMult]}
  n28 = ScalarParameter() {defaultValue=75, parameterName=CameraFade}
  n29 = OneMinus(None=SphereMask(
    A=WorldPosition().XYZ
    B=CameraPositionWS()
    Radius=StaticSwitchParameter(True=@n27.CameraFadeDistance, False=@n28) {parameterName=UseNiagaraDynamicParameter?}
    Hardness=Constant() {r=0.25}
   ))
  n30 = Multiply(
   A=Multiply(A=ComponentMask(None=@n20.XYZ Texture) {b=true}, B=Multiply(A=@n25, B=@n25))
   B=ShadingPathSwitch(Default=Multiply(A=@n26, B=@n29), Mobile=Constant() {r=1}) {desc=Migrated from FeatureLevelSwitch}
  )
  n31 = Multiply(A=Multiply(A=@n30, B=@n30), B=TextureSample() {texture=/Game/Effects/Fort_Effects/Textures/Ramps/T_Ramp_Smooth}.R)
  n32 = TextureCoordinate() {uTiling=4, vTiling=2}
  n33 = DynamicParameter() {parameterIndex=1, defaultValue=<rgba 0.1 0.1 1 0.8>, paramNames [XSpeed, YSpeed, TileMult, DistortTextureTiling]}
  n34 = StaticSwitchParameter(True=@n33.TileMult, False=ScalarParameter() {defaultValue=1, parameterName=TileMult}) {parameterName=UseNiagaraDynamicParameter?}
  n35 = DynamicParameter() {parameterIndex=1, defaultValue=<rgba 0.1 0.1 1 0.8>, paramNames [XSpeed, YSpeed, TileMult, DistortTextureTiling]}
  n36 = AppendVector(
   A=StaticSwitchParameter(True=@n33.XSpeed, False=ScalarParameter() {defaultValue=-0.1, parameterName=XSpeed2}) {parameterName=UseNiagaraDynamicParameter?}
   B=StaticSwitchParameter(True=@n35.YSpeed, False=ScalarParameter() {defaultValue=0.25, parameterName=YSpeed2}) {parameterName=UseNiagaraDynamicParameter?}
  )
  n37 = StaticSwitchParameter(
   True=ScalarParameter() {defaultValue=1, parameterName=SpeedMult}
   False=DynamicParameter() {defaultValue=<rgba 0.1 0.1 1 0.8>, paramNames [XSpeed,YSpeed,TileMult,DepthFade]}.XSpeed
  ) {defaultValue=true, parameterName=UseWithStaticMeshes?}
  n38 = TextureSampleParameter2D(UVs=Panner(Coordinate=Multiply(A=@n32, B=@n34), Speed=Multiply(A=@n36, B=@n37)) {speedX=-0.1, speedY=1}) {
   group=Textures
   parameterName=ExtraFogTexture_Local
   samplerType=SAMPLERTYPE_Grayscale
   texture=&LowResBlurredNoise
  }
  n39 = LinearInterpolate(Alpha=@n38.R) {constA=0.0039, constB=0.0117}
  n40 = StaticSwitchParameter(True=@n39, False=@n39) {defaultValue=true, parameterName=UseWithStaticMeshes?}
  n41 = ScalarParameter() {defaultValue=1, parameterName="Value_Morning (S)"}
  n42 = DynamicParameter() {parameterIndex=3, paramNames [Value_Morning, Value_Day, Value_Evening, Value_Night]}
  n43 = ScalarParameter() {defaultValue=1, parameterName="Value_Day (S)"}
  n44 = DynamicParameter() {parameterIndex=3, paramNames [Value_Morning, Value_Day, Value_Evening, Value_Night]}
  n45 = ScalarParameter() {defaultValue=1, parameterName="Value_Evening (S)"}
  n46 = DynamicParameter() {parameterIndex=3, paramNames [Value_Morning, Value_Day, Value_Evening, Value_Night]}
  n47 = ScalarParameter() {defaultValue=1, parameterName="Value_Night (S)"}
  n48 = DynamicParameter() {parameterIndex=3, paramNames [Value_Morning, Value_Day, Value_Evening, Value_Night]}
  Multiply(
   A=Multiply(A=Subtract(A=@n31, B=@n40), B=StaticSwitchParameter(True=@n23, False=@n24.A) {defaultValue=true, parameterName=UseWithStaticMeshes?})
   B=MaterialFunctionCall(
    Value_Morning=StaticSwitchParameter(True=@n41, False=@n42.Value_Morning) {defaultValue=true, parameterName=UseWithStaticMeshes?}
    Value_Day=StaticSwitchParameter(True=@n43, False=@n44.Value_Day) {defaultValue=true, parameterName=UseWithStaticMeshes?}
    Value_Evening=StaticSwitchParameter(True=@n45, False=@n46.Value_Evening) {defaultValue=true, parameterName=UseWithStaticMeshes?}
    Value_Night=StaticSwitchParameter(True=@n47, False=@n48.Value_Night) {defaultValue=true, parameterName=UseWithStaticMeshes?}
   ) {materialFunction=&Scalar_Time_of_Day}.Result
  )
 ]
}
