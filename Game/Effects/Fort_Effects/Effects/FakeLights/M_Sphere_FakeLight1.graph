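# Unlit Modulate-blend material for a fake light sphere, usable on mesh and Niagara mesh particles, that multiplies the scene behind it to mimic lighting.
# A fast, clamped sine offset adds flicker to a high brightness constant.
# A mask from camera vector against world Z and scaled scene depth blends neutral white toward that brightness.
# A final pixel-depth-based lerp can return the result to neutral white.
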Material /Game/Effects/Fort_Effects/Effects/FakeLights/M_Sphere_FakeLight1
graph {
 properties {
  bUsedWithMeshParticles=true
  bUsedWithNiagaraMeshParticles=true
  bUsedWithStaticLighting=true
  blendMode=BLEND_Modulate
  pixelDepthOffsetMode=PDOM_Legacy
  shadingModel=MSM_Unlit
  translucencyPass=MTP_BeforeDOF
  usedShadingModels=Unlit
 }
 nodes [
  ComponentMask(None=ScreenPosition().ViewportUV) {b=true}
  n1 = Constant() {r=30}
  n2 = Time()
  n3 = Clamp(None=Add(A=Sine(None=@n2) {period=0.4}, B=Constant() {r=5}), Min=Constant() {r=1}, Max=Constant() {r=10})
  n4 = CameraVectorWS()
  n5 = Constant3Vector() {constant=<rgba 0 0 1 0>}
  n6 = Clamp(None=Power(Base=DotProduct(A=@n4, B=@n5), Exp=Constant() {r=6}), Min=Constant(), Max=Constant() {r=1})
  n7 = SceneDepth()
  n8 = Constant() {r=0.0035}
  n9 = Clamp(None=Power(Base=Multiply(A=@n7.R, B=@n8), Exp=Constant() {r=1.5}), Min=Constant(), Max=Constant() {r=1})
  n10 = Divide(A=Subtract(B=PixelDepth().R), B=Constant() {r=30})
  LinearInterpolate(
   A=LinearInterpolate(A=Constant() {r=1}, B=Add(A=@n1, B=@n3), Alpha=Multiply(A=@n6, B=@n9))
   B=Constant() {r=1}
   Alpha=OneMinus(None=Clamp(None=@n10))
  )
  PixelDepth()
 ]
}
